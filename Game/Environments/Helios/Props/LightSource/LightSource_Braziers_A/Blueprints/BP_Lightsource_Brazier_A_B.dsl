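; Blueprint for a brazier light source prop, variant A_B, in the Helios environment set.
; Its construction script runs the parent's, then sets the StaticMesh1 component's relative location to a fixed height with a small random Y offset.
; The offset is unseeded, so it can change each time the construction script reruns.
; BeginPlay, overlap and Tick events are disabled stubs, so the graphs add no gameplay logic.
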(event EventBeginPlay
  (|CallParentFunction) ; This node is disabled and will not be called.\nDrag off pins to build functionality.
)

(event Collision|EventActorBeginOverlap
  (params (OtherActor :type Actor%20Object%20Reference))
  (|CallParentFunction OtherActor) ; This node is disabled and will not be called.\nDrag off pins to build functionality.
)

(event EventTick
  (params (DeltaSeconds :type Float%20%28single-precision%29))
  (|CallParentFunction DeltaSeconds) ; This node is disabled and will not be called.\nDrag off pins to build functionality.
)

(fn ConstructionScript
  (|CallParentFunction)
  (Transformation|SetRelativeLocation (Variables|Default|GetStaticMesh1) (Math|Vector|MakeVector 0.0 (Math|Random|RandomFloatinRange 0.1 0.9) 121.281914) false false))
